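# Niagara system for Season 21 front-end confetti: one GPU sprite emitter that continuously spawns small, multicolored flipbook pieces.
# Spawn stores a grid position as each piece's ground reference, then applies a squashed sphere location and curve-picked color.
# Wind, curl noise and a turbulence vector field move pieces.
# A height test fakes landing, zeroing vertical velocity, adding drag, freezing the flipbook and turning sprites from camera-facing to flat.
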NiagaraSystem /Game/Effects/Fort_Effects/FrontEnd/S21/NS_FrontEnd_S21_Confetti
dependencies {
 usedDataInterfaces [&NiagaraDataInterfaceCurve, &NiagaraDataInterfaceColorCurve, /Script/Niagara.NiagaraDataInterfaceVectorField]
 usedDynamicInputs []
 usedModules [&SystemState]
}
emitters {InteractiveConfetti001 {
  simTarget=GPUComputeSim
  emitterUpdateScript [
   &EmitterState()
   &SpawnBurst_Instantaneous(Spawn Count=7500, Use Spawn Probability=false) {enabled=false}
   &SpawnRate(SpawnRate=30, Use Spawn Probability=false)
   /Niagara/Modules/Spawn/Location/SpawnParticlesInGrid(X Count=5, Y Count=5, Z Count=1)
  ]
  particleSpawnScript [
   /Niagara/Modules/Spawn/Initialization/InitializeParticle(
    Color=&RandomRangeLinearColor(Minimum=<rgba 1 0.1164 0 1>, Maximum=<rgba 1 0.2736 0 1>)
    Lifetime=&RandomRangeFloat(Minimum=8, Maximum=10)
    Mass=5
    Module.Write Color=true
    Module.Write Lifetime=true
    Module.Write Mass=true
    Module.Write Material Random=true
    Module.Write Position=false
    Module.Write Scale=true
    Module.Write SpriteRotation=true
    Module.Write SpriteSize=true
    Position=&SimulationPosition()
    Sprite Rotation=&RandomRangeFloat(Minimum=0, Maximum=360)
    Sprite Size=&Vector2DFromFloat(Value=&RandomRangeFloat(Minimum=4, Maximum=6))
   )
   /SetVariables(Particles.Terrain Offset=10) {isSetParameters=true}
   /Niagara/Modules/Spawn/Location/V2/GridLocation(Module.Jitter Inside Grid Cells=true, Module.Override Random Seed=false, Offset:Vector3f=<1 1 0>, XYZ Dimensions:Vector3f=<250 100 1>)
   /SetVariables(Particles.GridPosition:NiagaraPosition=$Particles.Position) {isSetParameters=true}
   /SetVariables(Particles.Position=&Add_Vector(
     A=Linked() {linkedVariable {name=Particles.Position, type=NiagaraPosition}}
     B=&MakeVector(X=0, Y=0, Z=$Particles.Terrain Offset)
    )) {isSetParameters=true}
   /SetVariables(Particles.New Vector 3f:Vector3f=<0 0 1>) {isSetParameters=true}
   &ShapeLocation(
    MaskSpawnGroup=false
    Non Uniform Scale=<1 1 0.65>
    Offset=<0 300 200>
    Offset Mode=::NewEnumerator0
    Override Seed=false
    Sphere Radius=250
    UseOwnerScale=true
   )
   &CameraOffset(Camera Offset Amount=15)
   &SubUVAnimation(End Frame=36, SubUV Animation Mode:ENiagara_SubUVLookupMode=::NewEnumerator2) {module=SubUVAnimation001}
   &Color(Color=&ColorFromCurve(
     ColorCurve=DataInterface() {propertyValues {
       AlphaCurve {keys [<at 0 1>,<at 1 1>]}
       BlueCurve {keys [<at 0 0.8714>,<at 0.2389 0.8714>,<at 0.2977 0.9307>,<at 0.3526 0.7554>,<at 0.5986 0.8662>,<at 0.7596 0>,<at 0.8643 0>,<at 1 0.1624>]}
       GreenCurve {keys [<at 0 0.5149>,<at 0.2389 0.5149>,<at 0.2977 0>,<at 0.3526 0>,<at 0.5986 0.9583>,<at 0.7596 0.595>,<at 0.8643 0.5433>,<at 1 0.9583>]}
       RedCurve {keys [<at 0 0.0144>,<at 0.2389 0.0144>,<at 0.2977 1>,<at 0.3526 1>,<at 0.5986 0>,<at 0.7596 0.9583>,<at 0.8643 1>,<at 1 0>]}
      }}
     CurveIndex=$Particles.MaterialRandom
    )) {module=Color001}
   &Color(Color=&ColorFromCurve(
     ColorCurve=DataInterface() {propertyValues {
       AlphaCurve {keys [<at 0 1 RCIM_Cubic>,<at 1 1 RCIM_Cubic>]}
       BlueCurve {keys [
         <at 0 0 RCIM_Cubic>
         <at 0.25 0 RCIM_Cubic>
         {arriveTangent=2, interpMode=RCIM_Cubic, leaveTangent=2, time=0.5, value=0}
         <at 0.75 1 RCIM_Cubic>
         <at 1 0 RCIM_Cubic>
        ]}
       GreenCurve {keys [
         <at 0 0.018 RCIM_Cubic>
         {arriveTangent=1.2764, interpMode=RCIM_Cubic, leaveTangent=1.2764, time=0.25, value=0.4937}
         {arriveTangent=-0.7874, interpMode=RCIM_Cubic, leaveTangent=-0.7874, time=0.5, value=0.6562}
         {arriveTangent=-1.2764, interpMode=RCIM_Cubic, leaveTangent=-1.2764, time=0.75, value=0.1}
         <at 1 0.018 RCIM_Cubic>
        ]}
       RedCurve {keys [
         <at 0 0.5573 RCIM_Cubic>
         {arriveTangent=-0.8057, interpMode=RCIM_Cubic, leaveTangent=-0.8057, time=0.25, value=0.5885}
         {arriveTangent=-1.5635, interpMode=RCIM_Cubic, leaveTangent=-1.5635, tangentMode=RCTM_User, time=0.5, value=0.1544}
         {arriveTangent=0.8057, interpMode=RCIM_Cubic, leaveTangent=0.8057, time=0.75, value=0.0257}
         <at 1 0.5573 RCIM_Cubic>
        ]}
      }}
     CurveIndex=$Particles.MaterialRandom
    )) {enabled=false, module=Color002}
   /SetVariables(Particles.SubImageIndex=&RandomRangeFloat(Minimum=0, Maximum=35)) {isSetParameters=true}
   /SetVariables() {isSetParameters=true}
   &AddVelocity(Velocity=<-50 0 -50>)
   /SetVariables() {isSetParameters=true}
  ]
  particleUpdateScript [
   /SetVariables(Particles.Has Collided With Terrain:NiagaraBool=&SetBoolByFloatComparison(
     A=&Add_Float(
      A=&MakeFloatFromVector(Vector=Linked() {linkedVariable {name=Particles.GridPosition, type=NiagaraPosition}}, Channel=::NewEnumerator2)
      B=15
     )
     B=&MakeFloatFromVector(Vector=Linked() {linkedVariable {name=Particles.Position, type=NiagaraPosition}}, Channel=::NewEnumerator2)
    )) {isSetParameters=true}
   /SetVariables(Particles.SubUVAge:NiagaraFloat=&MakeCustomFloatFromBool(Bool=$Particles.Has Collided With Terrain, True Float=$Particles.SubUVAge, False Float=$Particles.NormalizedAge)) {isSetParameters=true}
   /SetVariables(Particles.Normalized Velocity:Vector3f=&NormalizeVector(Vector To Normalize=$Particles.Velocity)) {isSetParameters=true}
   /SetVariables(Particles.Is Active:NiagaraBool=&Boolean_OrOperation(
     Bool1=$Particles.Is Active
     Bool2=&SetBoolByFloatComparison(
      A=&MakeFloatFromVector(Vector=Linked() {linkedVariable {name=Particles.Position, type=NiagaraPosition}}, Channel=::NewEnumerator2)
      B=&MakeFloatFromVector(Vector=Linked() {linkedVariable {name=Particles.GridPosition, type=NiagaraPosition}}, Channel=::NewEnumerator2)
      Comparison Type=::NewEnumerator2
     )
    )) {isSetParameters=true}
   &WindForce(
    Wind Speed=&Multiply_VectorByFloat(
     Vector=&MakeVector(
      X=-600
      Y=0
      Z=&FloatFromCurve(FloatCurve=DataInterface() {propertyValues {Curve {keys [<at 0 1>,<at 0.25 -1>]}}}, Scale Curve=150)
     )
     Float=1
    )
    Wind Speed Scale=&MakeCustomFloatFromBool(Bool=$Particles.Has Collided With Terrain, True Float=0, False Float=0.1)
   )
   &CurlNoiseForce(
    Noise Frequency=4
    Noise Strength=&MakeCustomFloatFromBool(Bool=$Particles.Has Collided With Terrain, True Float=0, False Float=150)
    Pan Noise=true
    Pan Noise Field=<0.15 0 0.25>
   ) {module=CurlNoiseForce001}
   &Drag(
    Drag=&MakeCustomFloatFromBool(
     Bool=$Particles.Has Collided With Terrain
     True Float=&FloatFromCurve(FloatCurve=DataInterface() {propertyValues {Curve {keys [<at 0.1 0>,<at 0.25 0.95>,<at 0.35 1>,<at 0.5 0>]}}}, Scale Curve=15)
     False Float=0
    )
    Use Linear Drag=true
    Use Rotational Drag=false
   )
   /SetVariables(Particles.Sprite Orientation:Vector3f=&Lerp_Vector(
     A=$Particles.New Vector 3f
     B=&ReturnCameraVector(Camera Property=::NewEnumerator4, Coordinate Space=::Simulation, Origin Position=$Particles.Position, CameraQuery=$Emitter.CameraRead)
     Alpha=&MakeCustomFloatFromBool(Bool=$Particles.Has Collided With Terrain, True Float=0, False Float=1)
    )) {isSetParameters=true}
   &SolveForcesAndVelocity(Clamp Velocity=true, Limit Acceleration=false, Speed Limit=1500)
   &SubUVAnimation(
    End Frame=35
    SubUV Animation Mode:ENiagara_SubUVLookupMode=::NewEnumerator1
    SubUV Curve Index:NiagaraFloat=$Particles.SubUVAge
    SubUV Loop Count=4
   )
   &ScaleColor(
    Scale Alpha=&FloatFromCurve(FloatCurve=DataInterface() {propertyValues {Curve {keys [<at 0 0>,<at 0.1 1>,<at 0.9 0.95>,<at 1 0>]}}})
    ScaleA=true
    ScaleRGB=true
   )
   &ParticleState() {module=ParticleState001}
   /SetVariables(Particles.Velocity=&MakeVector(
     X=&MakeFloatFromVector(Vector=$Particles.Velocity, Channel=::NewEnumerator0)
     Y=&MakeFloatFromVector(Vector=$Particles.Velocity, Channel=::NewEnumerator1)
     Z=&MakeCustomFloatFromBool(
      Bool=$Particles.Has Collided With Terrain
      True Float=0
      False Float=&MakeFloatFromVector(Vector=$Particles.Velocity, Channel=::NewEnumerator2)
     )
    )) {isSetParameters=true}
   /Niagara/VectorFields/SampleVectorField(
    Sampled Vector Scale=0.5
    VectorField=DataInterface() {propertyValues {Field=/Game/Effects/VectorFields/Noisy_Turbulence_103, bTileX=true, bTileY=true, bTileZ=true}}
   )
   /Niagara/VectorFields/ApplyVectorField(Vector Field Intensity as Force=1500, Vector Field Intensity as Velocity=1, Vector Field Velocity Tightness=0.005)
  ]
  rendererClasses [&NiagaraSpriteRendererProperties]
  renderers [{rendererClass=&NiagaraSpriteRendererProperties}]
 }}
summary {
 systemName=NS_FrontEnd_S21_Confetti
 emitters [{emitterName=InteractiveConfetti001, simTarget=GPUComputeSim, rendererClasses [&NiagaraSpriteRendererProperties]}]
 userVariables [
  User.Alpha_AttractionStrength {type=NiagaraFloat, defaultValue=0, description=""}
  User.Player Position {type=Vector3f, description="", defaultValue {struct=Vector3f, value=<0 0 0>}}
 ]
}
